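; DJ booth Blueprint that syncs Yacht DJ booth playback across the mesh network and game clients.
; Root play/stop events set a UTC start time or stop sentinel, plus an audio UID on play.
; Edge nodes convert that UTC time to server world time, using -1 for stopped.
; Remote clients derive the elapsed playback offset from it, then fire ready-to-play with the audio UID, or stop playback.
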(event EventBeginPlay)

(event Custom|RootPlayYachtDJBooth
  (params (AudioUID :type String))
  (Variables|Default|SetAudioUID AudioUID)
  (Variables|Default|SetPlayUTCTime (Math|DateTime|UTCNow)))

(event Custom|RootStopYachtDJBooth
  (Variables|Default|SetPlayUTCTime (Variables|Default|GetStopUTCTime)))

(fn OnRep_PlayServerTimeSeconds
  (Networking|SwitchHasAuthority
    (:Remote
      (if (== (Variables|Default|GetPlayServerTimeSeconds) -1.0)
        (Variables|Default|SetVideoPlayStartTime -1.0)
        (Default|CallOnStopPlayback)
        (else
          (LocalVariables|SetCachedServerWorldTimeSeconds (Fortnite|Sync|Time|GetServerWorldTimeSeconds))
          (Fortnite|FortShippingLog 0 (Utilities|String|Append "DJBooth: PlayServerTimeSeconds replicated as: " (Utilities|String|ToString(Float) (Variables|Default|GetPlayServerTimeSeconds)) "
" "Client GetServerWorldTimeSeconds:" (Utilities|String|ToString(Float) (LocalVariables|GetCachedServerWorldTimeSeconds))))
          (bind _output_get (Variables|Default|SetVideoPlayStartTime (- (LocalVariables|GetCachedServerWorldTimeSeconds) (Variables|Default|GetPlayServerTimeSeconds))))
          (Default|CallOnReadytoPlay _output_get (Variables|Default|GetAudioUID))
          (Variables|Default|SetAudioUID ""))))))

(fn OnRep_PlayUTCTime
  (switch Utilities|FlowControl|Switch|SwitchonEMeshNetworkNodeType (MeshNetworkSubsystem|GetMeshNetworkNodeType (GameInstanceSubsystems|GetMeshNetworkSubsystem))
    (:Edge
      (if (Math|DateTime|Equal(DateTime) (Variables|Default|GetPlayUTCTime) (Variables|Default|GetStopUTCTime))
        (Variables|Default|SetPlayServerTimeSeconds -1.0)
        (else
          (Variables|Default|SetPlayServerTimeSeconds (- (Fortnite|Sync|Time|GetServerWorldTimeSeconds) (Math|Timespan|GetTotalSeconds (Math|DateTime|DateTime-DateTime (Math|DateTime|UTCNow) (Variables|Default|GetPlayUTCTime))))))))))

(fn ConstructionScript)
